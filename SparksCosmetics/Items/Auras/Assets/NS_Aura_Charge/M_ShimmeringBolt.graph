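# Translucent unlit Niagara sprite material for a shimmering lightning bolt in the Charge aura cosmetic.
# Plasma coil noise, randomly offset per particle, pans in opposite directions on each horizontal half.
# The bolt texture minus its center-scaled copy masks that noise into a rim shimmer, tinted by particle color through an emissive adjustment function.
# Bolt green-channel masks combine with particle alpha, a flicker dynamic parameter and tiled-pattern erosion.
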Material /SparksCosmetics/Items/Auras/Assets/NS_Aura_Charge/M_ShimmeringBolt
graph {
 properties {
  bAutomaticallySetUsageInEditor=false
  bUsedWithNiagaraSprites=true
  blendMode=BLEND_Translucent
  shadingModel=MSM_Unlit
  usedShadingModels=Unlit
 }
 nodes [
  n1 = ParticleColor()
  n2 = TextureCoordinate()
  n3 = ParticleRandom()
  n4 = AppendVector(A=@n3, B=@n3)
  n5 = Constant2Vector() {g=-0.5, r=0.5}
  n6 = Constant2Vector() {g=0.5, r=-0.5}
  n7 = Round(None=ComponentMask(None=TextureCoordinate()) {r=true})
  n8 = TextureSampleParameter2D(UVs=Panner(Coordinate=Add(A=@n2, B=@n4), Speed=LinearInterpolate(A=@n5, B=@n6, Alpha=@n7)) {speedX=-0.5, speedY=0.5}) {parameterName=Param, samplerType=SAMPLERTYPE_LinearGrayscale, texture=/Game/Effects/Fort_Effects/Emotes/UltraLight/Texture/T_PlasmaCoils}
  n9 = TextureSample() {samplerSource=SSM_Wrap_WorldGroupSettings, texture=&/T_Aura_ClampedBolt}
  n10 = Add(A=TextureCoordinate()) {constB=-0.5}
  n11 = ScalarParameter() {defaultValue=1.2, parameterName=CenterScale}
  n12 = TextureSample(UVs=Add(A=Multiply(A=@n10, B=@n11)) {constB=0.5}) {samplerSource=SSM_Wrap_WorldGroupSettings, texture=&T_Aura_ClampedBolt}
  n13 = Multiply(A=Power(Base=Subtract(A=@n9.RGB, B=@n12.RGB)) {constExponent=3}) {constB=16}
  n14 = Multiply(A=Multiply(A=Power(Base=@n8.RGB), B=Saturate(None=@n13)))
  n15 = TextureSample() {samplerSource=SSM_Wrap_WorldGroupSettings, texture=&T_Aura_ClampedBolt}
  MaterialFunctionCall(Emissive=Multiply(A=@n1.RGB, B=Multiply(A=@n14, B=@n15.RGB))) {materialFunction=&MF_FX_EmissiveAdjustments}
  n16 = DynamicParameter() {parameterIndex=1, defaultValue=<rgba 1 0 1 1>, paramNames [Color Flicker, Param2, Param3, Param4]}
  Multiply(A=Multiply(A=@n1.A, B=@n15.G), B=@n16.Color Flicker)
  Multiply(
   A=@n15.G
   B=SmoothStep(
    Max=DynamicParameter() {defaultValue=<rgba 0 0 1 1>, paramNames [Erosion,Param2,Parm3,Param4]}.Erosion
    Value=TextureSample() {texture=/Game/Effects/Fort_Effects/Textures/T_Tile_Pattern_01}.G
   ) {constMax=0}
  )
 ]
}
